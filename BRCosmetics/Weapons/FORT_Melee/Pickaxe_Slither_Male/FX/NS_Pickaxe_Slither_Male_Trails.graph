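# Niagara system for the swing trails of the male Slither pickaxe cosmetic.
# Four ribbon AnimTrail emitters use the shared anim-trail modules, each bound to a different bone or socket pair, with mostly narrower, shorter-lived trails down the stack.
# Each feeds randomized values into two dynamic material parameters and fades its faint white alpha over particle life.
# Ribbon width scaling over life is present but disabled in every emitter.
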NiagaraSystem /BRCosmetics/Weapons/FORT_Melee/Pickaxe_Slither_Male/FX/NS_Pickaxe_Slither_Male_Trails
dependencies {usedDataInterfaces [&NiagaraDataInterfaceCurve], usedDynamicInputs [], usedModules [&SystemState]}
emitters {
 AnimTrail {
  emitterUpdateScript [
   &EmitterState(Life Cycle Mode=::NewEnumerator1, Loop Behavior=::NewEnumerator1, Loop Delay=0.05, Loop Duration=0.5, UseLoopDelay=true)
   &FN_SpawnAnimTrail_v2()
  ]
  particleSpawnScript [
   /Game/Effects/Niagara/Modules/AnimTrail/FN_InitializeAnimTrail/FN_InitializeAnimTrailBoneIndecies(First Bone or Socket Index=1)
   /Game/Effects/Niagara/Modules/AnimTrail/FN_InitializeAnimTrail/FN_InitializeAnimTrail_v2(AnimTrailWidth=2.2, Color:LinearColor=<rgba 1 1 1 0.1>, Lifetime=0.19)
  ]
  particleUpdateScript [
   &FN_UpdateAnimTrail_v2()
   &SolveForcesAndVelocity(Clamp Velocity=false, Limit Acceleration=false)
   &DynamicMaterialParameters(
    Index 0 Param 1=&UniformRangedFloat(Minimum=0, Maximum=1)
    Index 0 Param 2=&UniformRangedFloat(Minimum=0, Maximum=1)
    Index 0 Param 3=1
    Index 0 Param 4=1
    Module.Param0WriteEnabled=true
    Module.Param1WriteEnabled=true
    Module.Param2WriteEnabled=true
    Module.Param3WriteEnabled=true
   )
   &ScaleRibbonWidth(Ribbon Width Scale=&FloatFromCurve(FloatCurve=DataInterface() {propertyValues {Curve {keys [
         {arriveTangent=-1, interpMode=RCIM_Cubic, time=0.8, value=1}
         {arriveTangent=-7.6765, interpMode=RCIM_Cubic, leaveTangent=-7.6765, tangentMode=RCTM_User, time=1, value=0}
        ]}}})) {enabled=false}
   &ScaleColor(
    Scale Alpha=&FloatFromCurve(FloatCurve=DataInterface() {propertyValues {Curve {keys [
         {arriveTangent=-5.203, interpMode=RCIM_Cubic, leaveTangent=-5.203, tangentMode=RCTM_User, time=0.3, value=1}
         {arriveTangent=-0.4707, interpMode=RCIM_Cubic, leaveTangent=-0.4707, tangentMode=RCTM_User, time=0.7, value=0.05}
         {arriveTangent=-0.0246, interpMode=RCIM_Cubic, leaveTangent=-0.0246, tangentMode=RCTM_User, time=0.9, value=0}
        ]}}})
    ScaleA=true
    ScaleRGB=true
   )
  ]
  rendererClasses [&NiagaraRibbonRendererProperties]
  renderers [{rendererClass=&NiagaraRibbonRendererProperties}]
 }
 AnimTrail001 {
  emitterUpdateScript [
   &EmitterState(Life Cycle Mode=::NewEnumerator1, Loop Behavior=::NewEnumerator1, Loop Delay=0.05, Loop Duration=0.5, UseLoopDelay=true)
   &FN_SpawnAnimTrail_v2()
  ]
  particleSpawnScript [
   &FN_InitializeAnimTrailBoneIndecies(First Bone or Socket Index=3, Second Bone or Socket Index=2)
   &FN_InitializeAnimTrail_v2(AnimTrailWidth=1.3, Color:LinearColor=<rgba 1 1 1 0.2>, Lifetime=0.19)
  ]
  particleUpdateScript [
   &FN_UpdateAnimTrail_v2()
   &SolveForcesAndVelocity(Clamp Velocity=false, Limit Acceleration=false)
   &DynamicMaterialParameters(
    Index 0 Param 1=&UniformRangedFloat(Minimum=0, Maximum=1)
    Index 0 Param 2=&UniformRangedFloat(Minimum=0, Maximum=1)
    Index 0 Param 3=1
    Index 0 Param 4=1
    Module.Param0WriteEnabled=true
    Module.Param1WriteEnabled=true
    Module.Param2WriteEnabled=true
    Module.Param3WriteEnabled=true
   )
   &ScaleRibbonWidth(Ribbon Width Scale=&FloatFromCurve(FloatCurve=DataInterface() {propertyValues {Curve {keys [
         {arriveTangent=-1, interpMode=RCIM_Cubic, time=0.8, value=1}
         {arriveTangent=-7.6765, interpMode=RCIM_Cubic, leaveTangent=-7.6765, tangentMode=RCTM_User, time=1, value=0}
        ]}}})) {enabled=false}
   &ScaleColor(
    Scale Alpha=&FloatFromCurve(FloatCurve=DataInterface() {propertyValues {Curve {keys [
         {arriveTangent=-5.203, interpMode=RCIM_Cubic, leaveTangent=-5.203, tangentMode=RCTM_User, time=0.3, value=1}
         {arriveTangent=-0.4707, interpMode=RCIM_Cubic, leaveTangent=-0.4707, tangentMode=RCTM_User, time=0.7, value=0.05}
         {arriveTangent=-0.0246, interpMode=RCIM_Cubic, leaveTangent=-0.0246, tangentMode=RCTM_User, time=0.9, value=0}
        ]}}})
    ScaleA=true
    ScaleRGB=true
   )
  ]
  rendererClasses [&NiagaraRibbonRendererProperties]
  renderers [{rendererClass=&NiagaraRibbonRendererProperties}]
 }
 AnimTrail002 {
  emitterUpdateScript [
   &EmitterState(Life Cycle Mode=::NewEnumerator1, Loop Behavior=::NewEnumerator1, Loop Delay=0.05, Loop Duration=0.5, UseLoopDelay=true)
   &FN_SpawnAnimTrail_v2()
  ]
  particleSpawnScript [
   &FN_InitializeAnimTrailBoneIndecies(First Bone or Socket Index=5, Second Bone or Socket Index=4)
   &FN_InitializeAnimTrail_v2(AnimTrailWidth=0.8, Color:LinearColor=<rgba 1 1 1 0.2>, Lifetime=0.17)
  ]
  particleUpdateScript [
   &FN_UpdateAnimTrail_v2()
   &SolveForcesAndVelocity(Clamp Velocity=false, Limit Acceleration=false)
   &DynamicMaterialParameters(
    Index 0 Param 1=&UniformRangedFloat(Minimum=0, Maximum=1)
    Index 0 Param 2=&UniformRangedFloat(Minimum=0, Maximum=1)
    Index 0 Param 3=1
    Index 0 Param 4=1
    Module.Param0WriteEnabled=true
    Module.Param1WriteEnabled=true
    Module.Param2WriteEnabled=true
    Module.Param3WriteEnabled=true
   )
   &ScaleRibbonWidth(Ribbon Width Scale=&FloatFromCurve(FloatCurve=DataInterface() {propertyValues {Curve {keys [
         {arriveTangent=-1, interpMode=RCIM_Cubic, time=0.8, value=1}
         {arriveTangent=-7.6765, interpMode=RCIM_Cubic, leaveTangent=-7.6765, tangentMode=RCTM_User, time=1, value=0}
        ]}}})) {enabled=false}
   &ScaleColor(
    Scale Alpha=&FloatFromCurve(FloatCurve=DataInterface() {propertyValues {Curve {keys [
         {arriveTangent=-5.203, interpMode=RCIM_Cubic, leaveTangent=-5.203, tangentMode=RCTM_User, time=0.3, value=1}
         {arriveTangent=-0.4707, interpMode=RCIM_Cubic, leaveTangent=-0.4707, tangentMode=RCTM_User, time=0.7, value=0.05}
         {arriveTangent=-0.0246, interpMode=RCIM_Cubic, leaveTangent=-0.0246, tangentMode=RCTM_User, time=0.9, value=0}
        ]}}})
    ScaleA=true
    ScaleRGB=true
   )
  ]
  rendererClasses [&NiagaraRibbonRendererProperties]
  renderers [{rendererClass=&NiagaraRibbonRendererProperties}]
 }
 AnimTrail003 {
  emitterUpdateScript [
   &EmitterState(Life Cycle Mode=::NewEnumerator1, Loop Behavior=::NewEnumerator1, Loop Delay=0.05, Loop Duration=0.5, UseLoopDelay=true)
   &FN_SpawnAnimTrail_v2()
  ]
  particleSpawnScript [
   &FN_InitializeAnimTrailBoneIndecies(First Bone or Socket Index=7, Second Bone or Socket Index=6)
   &FN_InitializeAnimTrail_v2(AnimTrailWidth=0.7, Color:LinearColor=<rgba 1 1 1 0.2>, Lifetime=0.15)
  ]
  particleUpdateScript [
   &FN_UpdateAnimTrail_v2()
   &SolveForcesAndVelocity(Clamp Velocity=false, Limit Acceleration=false)
   &DynamicMaterialParameters(
    Index 0 Param 1=&UniformRangedFloat(Minimum=0, Maximum=1)
    Index 0 Param 2=&UniformRangedFloat(Minimum=0, Maximum=1)
    Index 0 Param 3=1
    Index 0 Param 4=1
    Module.Param0WriteEnabled=true
    Module.Param1WriteEnabled=true
    Module.Param2WriteEnabled=true
    Module.Param3WriteEnabled=true
   )
   &ScaleRibbonWidth(Ribbon Width Scale=&FloatFromCurve(FloatCurve=DataInterface() {propertyValues {Curve {keys [
         {arriveTangent=-1, interpMode=RCIM_Cubic, time=0.8, value=1}
         {arriveTangent=-7.6765, interpMode=RCIM_Cubic, leaveTangent=-7.6765, tangentMode=RCTM_User, time=1, value=0}
        ]}}})) {enabled=false}
   &ScaleColor(
    Scale Alpha=&FloatFromCurve(FloatCurve=DataInterface() {propertyValues {Curve {keys [
         {arriveTangent=-5.203, interpMode=RCIM_Cubic, leaveTangent=-5.203, tangentMode=RCTM_User, time=0.3, value=1}
         {arriveTangent=-0.4707, interpMode=RCIM_Cubic, leaveTangent=-0.4707, tangentMode=RCTM_User, time=0.7, value=0.05}
         {arriveTangent=-0.0246, interpMode=RCIM_Cubic, leaveTangent=-0.0246, tangentMode=RCTM_User, time=0.9, value=0}
        ]}}})
    ScaleA=true
    ScaleRGB=true
   )
  ]
  rendererClasses [&NiagaraRibbonRendererProperties]
  renderers [{rendererClass=&NiagaraRibbonRendererProperties}]
 }
}
summary {
 systemName=NS_Pickaxe_Slither_Male_Trails
 emitters [
  {emitterName=AnimTrail, rendererClasses [&NiagaraRibbonRendererProperties]}
  {emitterName=AnimTrail001, rendererClasses [&NiagaraRibbonRendererProperties]}
  {emitterName=AnimTrail002, rendererClasses [&NiagaraRibbonRendererProperties]}
  {emitterName=AnimTrail003, rendererClasses [&NiagaraRibbonRendererProperties]}
 ]
 userVariables [
  User.RibbonIDOffset {type=NiagaraInt32}
  User.SkeletalMesh {
   type=NiagaraDataInterfaceSkeletalMesh
   defaultValue {
    struct=NiagaraExt_VariableValue_DataInterface
    value {dataInterface=&self:NiagaraDataInterfaceSkeletalMesh_0}
   }
  }
 ]
}
