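# Material function for the Slither male pickaxe that adds an animated metal shine and optional flame reflection to the input material's emissive.
# Projected UVs and a cutout mask are built from position in the vertex shader.
# Noise-distorted, panning shine and bump-offset, normal-distorted flame samples are tinted and scaled by FX-texture masking.
# The flame's layered-sine flicker lerps between identical FlameFadeRange.R inputs, so its brightness stays constant.
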MaterialFunction /BRCosmetics/Weapons/FORT_Melee/Pickaxe_Slither_Male/Materials/MF_Pickaxe_Slither_Male_FX
graph {nodes [
  n1 = GetMaterialAttributes(None=FunctionInput() {inputName=Material, inputType=FunctionInput_MaterialAttributes, sortPriority=2}) {attributeGetTypes [
    #1
    #2
    #3
    #4
   ]}
  n2 = NamedRerouteUsage()
  FunctionOutput(None=SetMaterialAttributes!(MaterialAttributes=@n1.MaterialAttributes, EmissiveColor=Add(A=@n1.EmissiveColor, B=@n2)) {attributeSetTypes [#1]})
  n3 = TextureSampleParameter2D() {
   group=SlitherMetalFX
   parameterName=FXMask
   samplerType=SAMPLERTYPE_Masks
   texture=/BRCosmetics/Weapons/FORT_Melee/Pickaxe_Slither_Male/Textures/T_Pickaxe_Slither_Male_FX
  }
  NamedRerouteDeclaration(None=@n3.R) {name=LogoMask, nodeColor=<rgba 1 1 1 1>}
  NamedRerouteDeclaration(None=@n1.Normal) {name=Normal, nodeColor=<rgba 0 0.3446 1 1>}
  n4 = WorldPosition()
  n5 = Divide(A=ComponentMask(None=TransformPosition(None=@n4.XYZ) {transformSourceType=TRANSFORMPOSSOURCE_World}) {b=true, g=true}) {constB=256}
  n6 = Constant2Vector() {g=-1, r=1}
  n7 = VectorParameter() {group=SlitherMetalFX, parameterName=UVScaleBias, defaultValue=<rgba 5 5 0.35 1.55>}
  n8 = Add(A=Multiply(A=Multiply(A=@n5, B=@n6), B=AppendVector(A=@n7.R, B=@n7.R)), B=AppendVector(A=@n7.B, B=@n7.B))
  n9 = Saturate(None=Floor(None=ComponentMask(None=@n8) {g=true}))
  n10 = VertexInterpolator(VS=AppendVector(A=Saturate(None=@n8), B=OneMinus(None=@n9) {desc=MaskOutArea}))
  NamedRerouteDeclaration(None=ComponentMask(None=@n10.PS) {b=true}) {name=ProjectedUVMask, nodeColor=<rgba 0.0312 1 0 1>}
  n11 = NamedRerouteUsage()
  n12 = Constant() {r=2}
  n13 = TextureSample(UVs=Panner(Coordinate=MaterialFunctionCall(UVs=@n11, Texture Scale=@n12) {materialFunction=&ScaleUVsByCenter}.UVs) {speedY=0.5}) {texture=/Game/Global/Textures/Default/Noise/T_Noise_Cloud}
  n14 = ComponentMask(None=VectorParameter() {group=SlitherMetalFX, parameterName=FlameDistortion, defaultValue=<rgba 0.6 0.4 0 1>}.RGB) {g=true, r=true}
  n15 = Multiply(A=AppendVector(A=@n13.G, B=@n13.G), B=@n14) {constB=0.7}
  n16 = Multiply(A=@n14) {constB=0.5}
  n17 = Multiply(
   A=BumpOffset(Coordinate=NamedRerouteUsage(), Height=ScalarParameter() {defaultValue=-20, group=SlitherMetalFX, parameterName=BumpOffset, sliderMax=-50}) {referencePlane=0}
   B=Constant2Vector() {g=-1, r=1}
  )
  n18 = VectorParameter() {group=SlitherMetalFX, parameterName=FlameOffsetSize, defaultValue=<rgba 0.2 0.2 1.5 1.5>}
  n19 = ComponentMask(None=@n18.RGB) {g=true, r=true}
  n20 = NamedRerouteUsage()
  n21 = Add(
   A=MaterialFunctionCall(UVs=Add(A=@n17, B=@n19), Texture Scale=AppendVector(A=@n18.B, B=@n18.B)) {materialFunction=&ScaleUVsByCenter}.UVs
   B=Multiply(
    A=ComponentMask(None=@n20) {g=true, r=true}
    B=ScalarParameter() {defaultValue=0.05, group=SlitherMetalFX, parameterName=FlameNormalDistortion}
   )
  )
  n22 = LinearInterpolate(Alpha=NamedRerouteUsage()) {constA=4, constB=0}
  n23 = TextureSampleParameter2D!(UVs=Add(A=Subtract(A=@n15, B=@n16), B=Saturate(None=@n21)), Bias=@n22, Apply View MipBias) {
   group=SlitherMetalFX
   mipValueMode=TMVM_MipBias
   parameterName=FlameMask
   samplerSource=SSM_Clamp_WorldGroupSettings
   texture=/Game/WaterAndWind/Textures/WeaponFireVectorMask
  }
  n24 = VectorParameter() {group=SlitherMetalFX, parameterName=FlameColorIntensity, defaultValue=<rgba 1 0.2862 0.0242 2>}
  n25 = Multiply(A=@n24.RGB, B=@n24.RGB) {constB=3}
  n26 = VectorParameter() {group=SlitherMetalFX, parameterName=FlameFadeRange, defaultValue=<rgba 0.02 1 0 0>}
  n27 = Time()
  n28 = VectorParameter() {group=SlitherMetalFX, parameterName=FlameFadeTimes, defaultValue=<rgba 0.23 1.2 3.33 0>}
  n29 = Sine(None=Multiply(A=@n27, B=@n28.R) {constB=0.23})
  n30 = Constant() {r=0.5}
  n31 = Multiply(A=Add(A=Multiply(A=@n29, B=@n30), B=@n30)) {constB=0.6}
  n32 = Sine(None=Multiply(A=@n27, B=@n28.G) {constB=1.2})
  n33 = Constant() {r=0.5}
  n34 = Multiply(A=Add(A=Multiply(A=@n32, B=@n33), B=@n33)) {constB=0.3}
  n35 = Sine(None=Multiply(A=@n27, B=@n28.B) {constB=3.33})
  n36 = Saturate(None=Add(A=Add(A=@n31, B=@n34), B=Multiply(A=@n35) {constB=0.125}))
  n37 = NamedRerouteUsage()
  n38 = Add(
   A=Rotator(Coordinate=NamedRerouteUsage(), Time=ScalarParameter() {group=SlitherMetalFX, parameterName=ShineRotation}) {speed=1}
   B=Multiply(A=ComponentMask(None=@n37) {g=true, r=true}, B=Constant() {r=0.05})
  )
  n39 = TextureSampleParameter2D!(UVs=Panner(Coordinate=@n38) {speedY=-0.025}, Apply View MipBias) {
   group=SlitherMetalFX
   parameterName=ShineDistorionNoise
   samplerType=SAMPLERTYPE_LinearColor
   texture=/Game/Weapons/WeaponSkins/Wraps/Textures/Beach/T_Noise_Water
  }
  n40 = Add(
   A=Panner(Coordinate=@n38, Speed=VectorParameter() {group=SlitherMetalFX, parameterName=ShineSpeed, defaultValue=<rgba 0 0.2 0 1>}.RGB)
   B=Multiply(A=@n39.G, B=ScalarParameter() {defaultValue=0.7, group=SlitherMetalFX, parameterName=ShineDistortion, sliderMax=1}) {constB=0.7}
  )
  n41 = VectorParameter() {group=SlitherMetalFX, parameterName=ShineColorIntensity, defaultValue=<rgba 0.5781 0.8026 1 2>}
  n42 = Multiply(
   A=Multiply(
    A=NamedRerouteUsage()
    B=TextureSampleParameter2D!(UVs=@n40, Bias=@n22, Apply View MipBias) {mipValueMode=TMVM_MipBias, parameterName=ShineGradient, texture=/Game/Effects/Fort_Effects/Textures/Grad/gradient_healed}.R
   )
   B=Multiply(A=@n41.RGB, B=@n41.RGB)
  ) {constB=5}
  n43 = Add(A=Multiply(A=Multiply(A=@n23.R, B=@n25), B=LinearInterpolate(A=@n26.R, B=@n26.R, Alpha=@n36)), B=@n42)
  n44 = Add(A=ScalarParameter() {defaultValue=0.05, group=SlitherMetalFX, parameterName=MaskLowerValue, sliderMax=1}, B=NamedRerouteUsage())
  NamedRerouteDeclaration(None=Multiply(A=StaticSwitchParameter(True=@n43, False=@n42) {parameterName=UseFlameReflection}, B=Saturate(None=@n44))) {name=Emissive, nodeColor=<rgba 1 0.1648 0 1>}
  NamedRerouteDeclaration(None=ComponentMask(None=@n10.PS) {g=true, r=true}) {name=ProjectedUVCoords, nodeColor=<rgba 0 0.6484 1 1>}
 ]}
